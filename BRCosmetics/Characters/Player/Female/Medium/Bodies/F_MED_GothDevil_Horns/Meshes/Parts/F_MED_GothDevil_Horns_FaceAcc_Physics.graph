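# Physics asset for the FaceAcc part of the female medium GothDevil Horns character cosmetic, built around an eight-bone back-hair chain.
# Capsule bodies on the pelvis, spine, neck, head, clavicles and upper arms approximate the torso, apparently as colliders for the hair.
# The dyn_Hair_Back capsules chain down from the head with narrow swing2 and varying twist limits, loosening after the root, while the body joints swing and twist freely.
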PhysicsAsset /BRCosmetics/Characters/Player/Female/Medium/Bodies/F_MED_GothDevil_Horns/Meshes/Parts/F_MED_GothDevil_Horns_FaceAcc_Physics
bodies {
 clavicle_l [{shapeType=Capsule, length=5.5517, radius=7.3175, shapeName=clavicle_l_capsule, center=<6.9223 0 0.217>, rotation=<rot -0.0001 -90 -91.7954>}]
 clavicle_r [{
   shapeType=Capsule
   length=5.6878
   radius=7.3175
   shapeName=clavicle_r_capsule
   center=<-6.0019 0 -0.1063>
   rotation=<rot -0.0001 -90 88.9855>
  }]
 dyn_Hair_Back_01 [{
   shapeType=Capsule
   length=5.6007
   radius=4.2147
   shapeName=dyn_Hair_Back_01_capsule
   center=<0.4811 -3.8564 -0.069>
   rotation=<rot 78.3227 -79.4314 3.9317>
  }]
 dyn_Hair_Back_02 [{
   shapeType=Capsule
   length=7.1379
   radius=5.1048
   shapeName=dyn_Hair_Back_02_capsule
   center=<3.0999 -3.4172 0.2639>
   rotation=<rot -77.7509 -177.8077 103.0948>
  }]
 dyn_Hair_Back_03 [{
   shapeType=Capsule
   length=11.932
   radius=5.2839
   shapeName=dyn_Hair_Back_03_capsule
   center=<0.4752 -5.3302 -0.2431>
   rotation=<rot -0.388 -0.9678 -90.2064>
  }]
 dyn_Hair_Back_04 [{
   shapeType=Capsule
   length=8.2015
   radius=4.8856
   shapeName=dyn_Hair_Back_04_capsule
   center=<-0.0253 -8.8829 -0.0173>
   rotation=<rot 81.8587 -169.6682 -80.1257>
  }]
 dyn_Hair_Back_05 [{
   shapeType=Capsule
   length=9.3126
   radius=3.8729
   shapeName=dyn_Hair_Back_05_capsule
   center=<-0.4217 -7.5087 0.5884>
   rotation=<rot 66.3048 -151.5347 -60.8752>
  }]
 dyn_Hair_Back_06 [{
   shapeType=Capsule
   length=9.5606
   radius=4.2116
   shapeName=dyn_Hair_Back_06_capsule
   center=<-0.4252 -5.638 0.6683>
   rotation=<rot 84.0258 154.1979 -116.2994>
  }]
 dyn_Hair_Back_07 [{
   shapeType=Capsule
   length=6.7411
   radius=4.2486
   shapeName=dyn_Hair_Back_07_capsule
   center=<-0.1978 -6.6253 1.8026>
   rotation=<rot 0 1.7333 -69.0108>
  }]
 dyn_Hair_Back_08 [{
   shapeType=Capsule
   length=5.5377
   radius=1.4118
   shapeName=dyn_Hair_Back_08_capsule
   center=<1.3062 -6.954 0.5001>
   rotation=<rot 0 -169.7999 -88.7011>
  }]
 head [{shapeType=Capsule, length=26.0416, radius=17.9176, shapeName=head_capsule, center=<8.8215 9.3196 -0.0436>, rotation=<rot -0.3902 -0.0011 0>}]
 neck_01 [{shapeType=Capsule, length=0.505, radius=0.505, shapeName=neck_01_capsule, rotation=<rot 0 -90 0>}]
 pelvis [{shapeType=Capsule, length=13.365, radius=16.635, shapeName=pelvis_capsule, center=<-3.0591 1.4406 0>}]
 spine_02 [{shapeType=Capsule, length=10.3569, radius=8.2801, shapeName=spine_02_capsule, rotation=<rot 0 -90 0>}]
 spine_04 [{shapeType=Capsule, length=9.3883, radius=9.2486, shapeName=spine_04_capsule, rotation=<rot 0 -90 0>}]
 spine_05 [{shapeType=Capsule, length=4.892, radius=21.7919, shapeName=spine_05_capsule, center=<2.0236 12.0551 0>, rotation=<rot 0 -90 0>}]
 upperarm_l [
  {
   shapeType=Capsule
   length=33.4871
   radius=5.2452
   shapeName=upperarm_l_capsule
   center=<15.142 0.0017 0.2617>
   rotation=<rot -0.0074 -89.9934 -90.99>
  }
  {
   shapeType=Capsule
   length=4.7852
   radius=12.4219
   shapeName=upperarm_l_capsule
   center=<4.5406 4.0485 8.8135>
   rotation=<rot -39.2843 0.0001 -0.0001>
  }
 ]
 upperarm_r [{
   shapeType=Capsule
   length=35.0128
   radius=5.2653
   shapeName=upperarm_r_capsule
   center=<-15.7143 0 0.3059>
   rotation=<rot -0.0001 -89.9999 91.1152>
  }]
}
constraints [
 {
  bone1Name=head
  bone2Name=neck_01
  swing1LimitDegrees=45
  swing1Motion=Free
  swing2LimitDegrees=45
  swing2Motion=Free
  twistLimitDegrees=45
  twistMotion=Free
 }
 {
  bone1Name=spine_05
  bone2Name=spine_04
  swing1LimitDegrees=45
  swing1Motion=Free
  swing2LimitDegrees=45
  swing2Motion=Free
  twistLimitDegrees=45
  twistMotion=Free
 }
 {
  bone1Name=spine_04
  bone2Name=spine_02
  swing1LimitDegrees=45
  swing1Motion=Free
  swing2LimitDegrees=45
  swing2Motion=Free
  twistLimitDegrees=45
  twistMotion=Free
 }
 {
  bone1Name=clavicle_l
  bone2Name=spine_05
  swing1LimitDegrees=45
  swing1Motion=Free
  swing2LimitDegrees=45
  swing2Motion=Free
  twistLimitDegrees=45
  twistMotion=Free
 }
 {
  bone1Name=upperarm_l
  bone2Name=clavicle_l
  swing1LimitDegrees=45
  swing1Motion=Free
  swing2LimitDegrees=45
  swing2Motion=Free
  twistLimitDegrees=45
  twistMotion=Free
 }
 {bone1Name=dyn_Hair_Back_01, bone2Name=head, swing1LimitDegrees=45, swing2LimitDegrees=1, twistLimitDegrees=5}
 {bone1Name=dyn_Hair_Back_02, bone2Name=dyn_Hair_Back_01, swing1LimitDegrees=45, swing2LimitDegrees=1, twistLimitDegrees=8}
 {bone1Name=dyn_Hair_Back_03, bone2Name=dyn_Hair_Back_02, swing1LimitDegrees=45, swing2LimitDegrees=1, twistLimitDegrees=25}
 {bone1Name=dyn_Hair_Back_04, bone2Name=dyn_Hair_Back_03, swing1LimitDegrees=45, swing2LimitDegrees=1, twistLimitDegrees=30}
 {bone1Name=dyn_Hair_Back_05, bone2Name=dyn_Hair_Back_04, swing1LimitDegrees=45, swing2LimitDegrees=10, twistLimitDegrees=20}
 {bone1Name=dyn_Hair_Back_06, bone2Name=dyn_Hair_Back_05, swing1LimitDegrees=45, swing2LimitDegrees=10, twistLimitDegrees=20}
 {bone1Name=dyn_Hair_Back_07, bone2Name=dyn_Hair_Back_06, swing1LimitDegrees=45, swing2LimitDegrees=10, twistLimitDegrees=20}
 {bone1Name=dyn_Hair_Back_08, bone2Name=dyn_Hair_Back_07, swing1LimitDegrees=45, swing2LimitDegrees=10, twistLimitDegrees=20}
 {
  bone1Name=neck_01
  bone2Name=spine_05
  swing1LimitDegrees=45
  swing1Motion=Free
  swing2LimitDegrees=45
  swing2Motion=Free
  twistLimitDegrees=45
  twistMotion=Free
 }
 {
  bone1Name=clavicle_r
  bone2Name=spine_05
  swing1LimitDegrees=45
  swing1Motion=Free
  swing2LimitDegrees=45
  swing2Motion=Free
  twistLimitDegrees=45
  twistMotion=Free
 }
 {
  bone1Name=upperarm_r
  bone2Name=clavicle_r
  swing1LimitDegrees=45
  swing1Motion=Free
  swing2LimitDegrees=45
  swing2Motion=Free
  twistLimitDegrees=45
  twistMotion=Free
 }
]
