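# Post-process style AnimBP for the GothDevil Horns female face accessory, driving its hair secondary motion.
# Copied body pose runs through Control Rig and left-hair AnimDynamics (cached), then back-hair AnimDynamics and RigidBody with state-driven gravity.
# Backbling, downed and skydive states damp back-hair simulation toward the cached pose and offset bones; disabled dynamics uses that cache.
# Final bone offsets and spine constraints always shape the back hair.
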AnimBlueprint /BRCosmetics/Characters/Player/Female/Medium/Bodies/F_MED_GothDevil_Horns/Meshes/Parts/F_MED_GothDevil_Horns_FaceAcc_AnimBP
target_skeleton=&/F_MED_GothDevil_Horns_FaceAcc_Skeleton
graphs [
 AnimGraph {nodes [
   n1 = Use cached pose 'HairDynPose'()
   n2 = Use cached pose 'HairCachePose'()
   n3 = Get IsWearingPackBool or DBNO or DBNOCarrying() -> (IsWearingPackBool or DBNO or DBNOCarrying:double)
   n4 = "Transform (Modify) Bone - Bone: dyn_Hair_Back_06"!(
    ComponentPose=Local To Component(LocalPose=Two Way Blend(A=@n1, B=@n2, Alpha=@n3) {blendNode {alphaBoolBlend {blendOption=Linear}}})
    bAlphaBoolEnabled:bool=True
    Translation="-6,0,0"
    Rotation="0,0,0"
    Scale="0.6,1,1"
   ) -> (Pose) {
    alphaInputType=Bool
    rotationMode=BMM_Additive
    rotationSpace=BCS_BoneSpace
    scaleMode=BMM_Additive
    scaleSpace=BCS_BoneSpace
    translationMode=BMM_Additive
    translationSpace=BCS_BoneSpace
    alphaBoolBlend {blendOption=Linear}
    boneToModify {boneName=dyn_Hair_Back_06}
    scale {x=0.6}
    translation {x=-6}
   }
   n5 = "Transform (Modify) Bone - Bone: dyn_Hair_Back_05"!(ComponentPose=@n4, bAlphaBoolEnabled:bool=True, Translation="0,0,0", Rotation="0,20,0", Scale="1,1,1") -> (Pose) {
    alphaInputType=Bool
    rotationMode=BMM_Additive
    rotationSpace=BCS_BoneSpace
    scaleMode=BMM_Additive
    scaleSpace=BCS_BoneSpace
    translationMode=BMM_Additive
    translationSpace=BCS_BoneSpace
    alphaBoolBlend {blendOption=Linear}
    boneToModify {boneName=dyn_Hair_Back_05}
    rotation {yaw=20}
   }
   n6 = Blend Poses by bool(
    BlendPose_0=Component To Local(ComponentPose=@n5)
    BlendPose_1=Use cached pose 'HairDynPose'()
    bActiveValue=Get Is Dive or Dive Floating or Wearing or DBNO() -> (Is Dive or Dive Floating or Wearing or DBNO:bool)
   ) {blendType=Linear}
   n7 = Use cached pose 'HairCachePose'()
   n8 = Get bAnimDynamicsEnabled() -> (bAnimDynamicsEnabled:bool)
   n9 = "Transform (Modify) Bone - Bone: dyn_Hair_Back_04"!(
    ComponentPose=Local To Component(LocalPose=Blend Poses by bool(BlendPose_0=@n6, BlendPose_1=@n7, bActiveValue=@n8) {blendType=Linear})
    bAlphaBoolEnabled:bool=True
    Translation="-2,0,0"
    Rotation="0,5,0"
    Scale="0.6,1,1"
   ) -> (Pose) {
    alphaInputType=Bool
    rotationMode=BMM_Additive
    rotationSpace=BCS_BoneSpace
    scaleMode=BMM_Additive
    scaleSpace=BCS_BoneSpace
    translationMode=BMM_Additive
    translationSpace=BCS_BoneSpace
    alphaBoolBlend {blendOption=Linear}
    boneToModify {boneName=dyn_Hair_Back_04}
    rotation {yaw=5}
    scale {x=0.6}
    translation {x=-2}
   }
   n10 = "Transform (Modify) Bone - Bone: dyn_Hair_Back_07"!(ComponentPose=@n9, bAlphaBoolEnabled:bool=True, Translation="-4,0,0", Rotation="0,-10,0", Scale="0.6,1,1") -> (Pose) {
    alphaInputType=Bool
    rotationMode=BMM_Additive
    rotationSpace=BCS_BoneSpace
    scaleMode=BMM_Additive
    scaleSpace=BCS_BoneSpace
    translationMode=BMM_Additive
    translationSpace=BCS_BoneSpace
    alphaBoolBlend {blendOption=Linear}
    boneToModify {boneName=dyn_Hair_Back_07}
    rotation {yaw=-10}
    scale {x=0.6}
    translation {x=-4}
   }
   n11 = "Constraint - dyn_Hair_Back_05 (PP)"!(ComponentPose=@n10, Alpha=1, ConstraintWeights_0=0.5, ConstraintWeights_1:float=0.5) -> (Pose) {
    alphaBoolBlend {blendOption=Linear}
    boneToModify {boneName=dyn_Hair_Back_05}
    constraintSetup [{targetBone {boneName=spine_03}}, {targetBone {boneName=dyn_Hair_Back_04}}]
    constraintWeights [1,1]
   }
   n12 = "Constraint - dyn_Hair_Back_07 (PP)"!(ComponentPose=@n11, Alpha=1, ConstraintWeights_0=0.5, ConstraintWeights_1:float=0.5) -> (Pose) {
    alphaBoolBlend {blendOption=Linear}
    boneToModify {boneName=dyn_Hair_Back_07}
    constraintSetup [{targetBone {boneName=spine_03}}, {targetBone {boneName=dyn_Hair_Back_06}}]
    constraintWeights [1,1]
   }
   n13 = "Constraint - dyn_Hair_Back_08 (PP)"!(ComponentPose=@n12, Alpha=1, ConstraintWeights_0=0.5, ConstraintWeights_1:float=0.5) -> (Pose) {
    alphaBoolBlend {blendOption=Linear}
    boneToModify {boneName=dyn_Hair_Back_08}
    constraintSetup [{targetBone {boneName=spine_03}}, {targetBone {boneName=dyn_Hair_Back_07}}]
    constraintWeights [1,1]
   }
   n14 = "Transform (Modify) Bone - Bone: dyn_Hair_Back_05"!(ComponentPose=@n13, Alpha:float=1, Translation="0,2,0", Rotation="0,0,0", Scale="1,1,1") -> (Pose) {
    rotationMode=BMM_Additive
    rotationSpace=BCS_BoneSpace
    scaleMode=BMM_Additive
    scaleSpace=BCS_BoneSpace
    translationMode=BMM_Additive
    translationSpace=BCS_BoneSpace
    alphaBoolBlend {blendOption=Linear}
    boneToModify {boneName=dyn_Hair_Back_05}
    translation {y=2}
   }
   n15 = "Transform (Modify) Bone - Bone: dyn_Hair_Back_07"!(ComponentPose=@n14, Alpha:float=1, Translation="0,3,0", Rotation="0,0,-5", Scale="1,1,1") -> (Pose) {
    rotationMode=BMM_Additive
    rotationSpace=BCS_BoneSpace
    scaleMode=BMM_Additive
    scaleSpace=BCS_BoneSpace
    translationMode=BMM_Additive
    translationSpace=BCS_BoneSpace
    alphaBoolBlend {blendOption=Linear}
    boneToModify {boneName=dyn_Hair_Back_07}
    rotation {roll=-5}
    translation {y=3}
   }
   Output Pose(Result=Component To Local(ComponentPose=@n15)) {layerGroup=None}
   n16 = Control Rig(Source=Copy Pose From Mesh(SourceMeshComponent=Get Character Mesh() -> (Character Mesh:Skeletal Mesh Component Object Reference)) {bUseAttachedParent=true}) {
    alphaBoolBlend {blendOption=Linear}
    controlRigAssetReference {blueprintRigClass=&/F_MED_GothDevil_Horns_FaceAcc_CtrlRig.F_MED_GothDevil_Horns_FaceAcc_CtrlRig_C}
   }
   n17 = Anim Dynamics - Bone: dyn_Hair_l_01(ComponentPose=Local To Component(LocalPose=@n16), Alpha=0.2, ExternalForce="0,0,0") {
    angularDampingOverride=0.99
    angularSpringConstant=75
    bAngularSpring=true
    bGravityOverrideInSimSpace=true
    bOverrideAngularDamping=true
    bUseGravityOverride=true
    simulationSpace=BoneRelative
    alphaBoolBlend {blendOption=Linear}
    boundBone {boneName=dyn_Hair_l_01}
    componentAppliedLinearAccClamp=<3500 3500 3000>
    componentLinearAccScale=<0.5 0.5 0.5>
    componentLinearVelScale=<0.65 0.65 0.65>
    gravityOverride {z=-980}
    physicsBodyDefinitions [{
      collisionType=CustomSphere
      sphereCollisionRadius=2
      boundBone {boneName=dyn_Hair_l_01}
      boxExtents=<40 40 40>
      constraintSetup {angularLimitsMax=<0 1 20>, angularLimitsMin=<-8 -1 -15>, angularTarget=<0 0 0>}
      localJointOffset=<-2 -5 0>
     }]
    relativeSpaceBone {boneName=gravity}
   }
   n18 = Anim Dynamics - Bone: dyn_Hair_l_02(ComponentPose=@n17, Alpha=1, ExternalForce="0,0,0") {
    angularDampingOverride=0.99
    angularSpringConstant=75
    bAngularSpring=true
    bGravityOverrideInSimSpace=true
    bOverrideAngularDamping=true
    bUseGravityOverride=true
    bUseSphericalLimits=true
    simulationSpace=BoneRelative
    alphaBoolBlend {blendOption=Linear}
    boundBone {boneName=dyn_Hair_l_02}
    componentAppliedLinearAccClamp=<3500 3500 3000>
    componentLinearAccScale=<0.5 0.5 0.5>
    componentLinearVelScale=<0.65 0.65 0.65>
    gravityOverride {z=-980}
    physicsBodyDefinitions [{
      collisionType=CustomSphere
      sphereCollisionRadius=2
      boundBone {boneName=dyn_Hair_l_02}
      boxExtents=<25 25 25>
      constraintSetup {angularLimitsMax=<0 1 20>, angularLimitsMin=<-8 -1 -30>, angularTarget=<0 0 0>}
      localJointOffset=<0 -5 0>
     }]
    relativeSpaceBone {boneName=gravity}
    sphericalLimits [{limitRadius=7, drivingBone {boneName=head}, sphereLocalOffset=<0 0 0>}]
   }
   n19 = Anim Dynamics - Bone: dyn_Hair_l_03(ComponentPose=@n18, Alpha=1, ExternalForce="0,0,0") {
    angularDampingOverride=0.99
    angularSpringConstant=75
    bAngularSpring=true
    bGravityOverrideInSimSpace=true
    bOverrideAngularDamping=true
    bUseGravityOverride=true
    simulationSpace=BoneRelative
    alphaBoolBlend {blendOption=Linear}
    boundBone {boneName=dyn_Hair_l_03}
    componentAppliedLinearAccClamp=<3500 3500 3000>
    componentLinearAccScale=<0.5 0.5 0.5>
    componentLinearVelScale=<0.65 0.65 0.65>
    gravityOverride {z=-980}
    physicsBodyDefinitions [{
      collisionType=CustomSphere
      sphereCollisionRadius=2
      boundBone {boneName=dyn_Hair_l_03}
      boxExtents=<23 23 23>
      constraintSetup {angularLimitsMax=<0 1 30>, angularLimitsMin=<-8 -1 -15>, angularTarget=<0 0 0>}
      localJointOffset=<0 -5 0>
     }]
    relativeSpaceBone {boneName=gravity}
   }
   n20 = Anim Dynamics - Bone: dyn_Hair_l_04(ComponentPose=@n19, Alpha=1, ExternalForce="0,0,0") {
    angularDampingOverride=0.99
    angularSpringConstant=75
    bAngularSpring=true
    bGravityOverrideInSimSpace=true
    bOverrideAngularDamping=true
    bUseGravityOverride=true
    simulationSpace=BoneRelative
    alphaBoolBlend {blendOption=Linear}
    boundBone {boneName=dyn_Hair_l_04}
    componentAppliedLinearAccClamp=<3500 3500 3000>
    componentLinearAccScale=<0.5 0.5 0.5>
    componentLinearVelScale=<0.65 0.65 0.65>
    gravityOverride {z=-980}
    physicsBodyDefinitions [{
      collisionType=CustomSphere
      sphereCollisionRadius=2
      boundBone {boneName=dyn_Hair_l_04}
      boxExtents=<23 23 23>
      constraintSetup {angularLimitsMax=<0 1 30>, angularLimitsMin=<-8 -1 -15>, angularTarget=<0 0 0>}
      localJointOffset=<0 -5 0>
     }]
    relativeSpaceBone {boneName=gravity}
   }
   n21 = Anim Dynamics - Bone: dyn_Hair_l_01a(ComponentPose=@n20, Alpha=1, ExternalForce="0,0,0") {
    angularDampingOverride=0.99
    angularSpringConstant=75
    bAngularSpring=true
    bGravityOverrideInSimSpace=true
    bOverrideAngularDamping=true
    bUseGravityOverride=true
    simulationSpace=BoneRelative
    alphaBoolBlend {blendOption=Linear}
    boundBone {boneName=dyn_Hair_l_01a}
    componentAppliedLinearAccClamp=<3500 3500 3000>
    componentLinearAccScale=<0.5 0.5 0.5>
    componentLinearVelScale=<0.65 0.65 0.65>
    gravityOverride {z=-980}
    physicsBodyDefinitions [{
      collisionType=CustomSphere
      sphereCollisionRadius=2
      boundBone {boneName=dyn_Hair_l_01a}
      boxExtents=<10 8 10>
      constraintSetup {angularLimitsMax=<15 2 15>, angularLimitsMin=<-15 -2 -15>, angularTarget=<0 0 0>}
      localJointOffset=<0 -3 0>
     }]
    relativeSpaceBone {boneName=gravity}
   }
   n22 = Anim Dynamics - Bone: dyn_Hair_l_03a(ComponentPose=@n21, Alpha=1, ExternalForce="0,0,0") {
    angularDampingOverride=0.99
    angularSpringConstant=75
    bAngularSpring=true
    bGravityOverrideInSimSpace=true
    bOverrideAngularDamping=true
    bUseGravityOverride=true
    simulationSpace=BoneRelative
    alphaBoolBlend {blendOption=Linear}
    boundBone {boneName=dyn_Hair_l_03a}
    componentAppliedLinearAccClamp=<3500 3500 3000>
    componentLinearAccScale=<0.5 0.5 0.5>
    componentLinearVelScale=<0.65 0.65 0.65>
    gravityOverride {z=-980}
    physicsBodyDefinitions [{
      collisionType=CustomSphere
      sphereCollisionRadius=2
      boundBone {boneName=dyn_Hair_l_03a}
      boxExtents=<10 8 10>
      constraintSetup {angularLimitsMax=<15 2 15>, angularLimitsMin=<-15 -2 -15>, angularTarget=<0 0 0>}
      localJointOffset=<0 -3 0>
     }]
    relativeSpaceBone {boneName=gravity}
   }
   n23 = Anim Dynamics - Bone: dyn_Hair_l_03b(ComponentPose=@n22, Alpha=1, ExternalForce="0,0,0") {
    angularDampingOverride=0.99
    angularSpringConstant=75
    bAngularSpring=true
    bGravityOverrideInSimSpace=true
    bOverrideAngularDamping=true
    bUseGravityOverride=true
    simulationSpace=BoneRelative
    alphaBoolBlend {blendOption=Linear}
    boundBone {boneName=dyn_Hair_l_03b}
    componentAppliedLinearAccClamp=<3500 3500 3000>
    componentLinearAccScale=<0.5 0.5 0.5>
    componentLinearVelScale=<0.65 0.65 0.65>
    gravityOverride {z=-980}
    physicsBodyDefinitions [{
      collisionType=CustomSphere
      sphereCollisionRadius=2
      boundBone {boneName=dyn_Hair_l_03b}
      boxExtents=<10 8 10>
      constraintSetup {angularLimitsMax=<15 2 15>, angularLimitsMin=<-15 -2 -15>, angularTarget=<0 0 0>}
      localJointOffset=<0 -3 0>
     }]
    relativeSpaceBone {boneName=gravity}
   }
   Save cached pose 'HairCachePose'(Pose=Component To Local(ComponentPose=@n23)) {cacheName=HairCachePose}
   n24 = Anim Dynamics - Bone: dyn_Hair_Back_02a(ComponentPose=Local To Component(LocalPose=Use cached pose 'HairCachePose'()), Alpha=1, ExternalForce="0,0,0") {
    angularDampingOverride=0.9999
    angularSpringConstant=75
    bAngularSpring=true
    bGravityOverrideInSimSpace=true
    bOverrideAngularDamping=true
    bUseGravityOverride=true
    simulationSpace=BoneRelative
    alphaBoolBlend {blendOption=Linear}
    boundBone {boneName=dyn_Hair_Back_02a}
    componentAppliedLinearAccClamp=<3500 3500 3000>
    componentLinearAccScale=<0.5 0.5 0.5>
    componentLinearVelScale=<0.65 0.65 0.65>
    gravityOverride {z=-980}
    physicsBodyDefinitions [{
      collisionType=CustomSphere
      sphereCollisionRadius=2
      boundBone {boneName=dyn_Hair_Back_02a}
      boxExtents=<4 2 4>
      constraintSetup {angularLimitsMax=<15 2 30>, angularLimitsMin=<-15 -2 -5>, angularTarget=<0 0 0>}
      localJointOffset=<0 -3 0>
     }]
    relativeSpaceBone {boneName=root}
   }
   n25 = Anim Dynamics - Bone: dyn_Hair_Back_02b(ComponentPose=@n24, Alpha=1, ExternalForce="0,0,0") {
    angularDampingOverride=0.9999
    angularSpringConstant=75
    bAngularSpring=true
    bGravityOverrideInSimSpace=true
    bOverrideAngularDamping=true
    bUseGravityOverride=true
    simulationSpace=BoneRelative
    alphaBoolBlend {blendOption=Linear}
    boundBone {boneName=dyn_Hair_Back_02b}
    componentAppliedLinearAccClamp=<3500 3500 3000>
    componentLinearAccScale=<0.5 0.5 0.5>
    componentLinearVelScale=<0.65 0.65 0.65>
    gravityOverride {z=-980}
    physicsBodyDefinitions [{
      collisionType=CustomSphere
      sphereCollisionRadius=2
      boundBone {boneName=dyn_Hair_Back_02b}
      boxExtents=<4 2 4>
      constraintSetup {angularLimitsMax=<15 2 30>, angularLimitsMin=<-15 -2 -5>, angularTarget=<0 0 0>}
      localJointOffset=<0 -5 0>
     }]
    relativeSpaceBone {boneName=root}
   }
   n26 = Anim Dynamics - Bone: dyn_Hair_Back_02c(ComponentPose=@n25, Alpha=1, ExternalForce="0,0,0") {
    angularDampingOverride=0.9999
    angularSpringConstant=75
    bAngularSpring=true
    bGravityOverrideInSimSpace=true
    bOverrideAngularDamping=true
    bUseGravityOverride=true
    simulationSpace=BoneRelative
    alphaBoolBlend {blendOption=Linear}
    boundBone {boneName=dyn_Hair_Back_02c}
    componentAppliedLinearAccClamp=<3500 3500 3000>
    componentLinearAccScale=<0.5 0.5 0.5>
    componentLinearVelScale=<0.65 0.65 0.65>
    gravityOverride {z=-980}
    physicsBodyDefinitions [{
      collisionType=CustomSphere
      sphereCollisionRadius=2
      boundBone {boneName=dyn_Hair_Back_02c}
      boxExtents=<4 2 4>
      constraintSetup {angularLimitsMax=<15 2 30>, angularLimitsMin=<-15 -2 -5>, angularTarget=<0 0 0>}
      localJointOffset=<0 -5 0>
     }]
    relativeSpaceBone {boneName=root}
   }
   n27 = Anim Dynamics - Bone: dyn_Hair_Back_02d(ComponentPose=@n26, Alpha=1, ExternalForce="0,0,0") {
    angularDampingOverride=0.9999
    angularSpringConstant=75
    bAngularSpring=true
    bGravityOverrideInSimSpace=true
    bOverrideAngularDamping=true
    bUseGravityOverride=true
    simulationSpace=BoneRelative
    alphaBoolBlend {blendOption=Linear}
    boundBone {boneName=dyn_Hair_Back_02d}
    componentAppliedLinearAccClamp=<3500 3500 3000>
    componentLinearAccScale=<0.5 0.5 0.5>
    componentLinearVelScale=<0.65 0.65 0.65>
    gravityOverride {z=-980}
    physicsBodyDefinitions [{
      collisionType=CustomSphere
      sphereCollisionRadius=2
      boundBone {boneName=dyn_Hair_Back_02d}
      boxExtents=<4 2 4>
      constraintSetup {angularLimitsMax=<15 2 30>, angularLimitsMin=<-15 -2 -5>, angularTarget=<0 0 0>}
      localJointOffset=<0 -5 0>
     }]
    relativeSpaceBone {boneName=root}
   }
   n28 = Anim Dynamics - Bone: dyn_Hair_Back_02e(ComponentPose=@n27, Alpha=1, ExternalForce="0,0,0") {
    angularDampingOverride=0.9999
    angularSpringConstant=75
    bAngularSpring=true
    bGravityOverrideInSimSpace=true
    bOverrideAngularDamping=true
    bUseGravityOverride=true
    simulationSpace=BoneRelative
    alphaBoolBlend {blendOption=Linear}
    boundBone {boneName=dyn_Hair_Back_02e}
    componentAppliedLinearAccClamp=<3500 3500 3000>
    componentLinearAccScale=<0.5 0.5 0.5>
    componentLinearVelScale=<0.65 0.65 0.65>
    gravityOverride {z=-980}
    physicsBodyDefinitions [{
      collisionType=CustomSphere
      sphereCollisionRadius=2
      boundBone {boneName=dyn_Hair_Back_02e}
      boxExtents=<4 2 4>
      constraintSetup {angularLimitsMax=<15 2 30>, angularLimitsMin=<-15 -2 -5>, angularTarget=<0 0 0>}
      localJointOffset=<0 -5 0>
     }]
    relativeSpaceBone {boneName=root}
   }
   n29 = RigidBody!(
    ComponentPose=@n28
    Alpha
    OverrideWorldGravity:Vector=Get Gravity Override Cape() -> (Gravity Override Cape:Vector)
    ExternalForce
    ComponentLinearAccScale:Vector="0.4,0.25,0.2"
    ComponentLinearVelScale:Vector="0.45,0.2,0.15"
   ) -> (Pose) {
    bClampLinearTranslationLimitToRefPose=true
    bOverrideWorldGravity=true
    overridePhysicsAsset=&/F_MED_GothDevil_Horns_FaceAcc_Physics
    simulationSpace=BaseBoneSpace
    alphaBoolBlend {blendOption=Linear}
    baseBoneRef {boneName=spine_05}
    componentAppliedLinearAccClamp=<3000 1500 1000>
    componentLinearAccScale=<0.1 0.1 0.1>
    componentLinearVelScale=<0.1 0.1 0.1>
    overrideWorldGravity {z=-980}
    simSpaceSettings {
     maxAngularAcceleration=1000
     maxAngularVelocity=10
     maxLinearAcceleration=700
     maxLinearVelocity=500
     velocityScaleZ=0
     worldAlpha=0.1
     externalLinearDragV=<0.5 0.5>
    }
   }
   Save cached pose 'HairDynPose'(Pose=Component To Local(ComponentPose=@n29)) {cacheName=HairDynPose}
  ]}
 EventGraph {nodes [
   Event BlueprintUpdateAnimation()
   > n1 = Sequence!(execute) -> (then_0, then_1, then_2:Exec, then_3:Exec, then_4:Exec)
   > n2 = Gravity Override State Function DEPRECATED(CharacterPartAnimInstance=Self-Reference(), GameTimeInSeconds=GetGameTimeInSeconds(), GravityOverrideParams=Get Gravity Override Params())
   > Set Gravity Override Cape(Gravity Override Cape:Vector=@n2) -> (Output_Get:Vector)
   n3 = Get bIsDBNOCarrying() -> (bIsDBNOCarrying:bool)
   n4 = Get isWearingPackBool() -> (isWearingPackBool:bool)
   n5 = Get bIsDBNO()
   Set IsWearingPackBool or DBNO or DBNOCarrying(
    execute=@n1.then_2
    IsWearingPackBool or DBNO or DBNOCarrying:double=SelectFloat(A=0.9, B=0, bPickA=OR Boolean!(A=@n4, B=@n5, C:bool=@n3) -> (ReturnValue))
   ) -> (Output_Get:double)
   Set Is Dive or Dive Floating or Wearing or DBNO(
    execute=@n1.then_3
    Is Dive or Dive Floating or Wearing or DBNO:bool=OR Boolean!(A=Get bIsSkydiveDiveMode(), B=Get bIsSkydiveFloating(), C=Get isWearingPackBool(), D:bool=Get bIsDBNO(), E:bool=Get bIsDBNOCarrying()) -> (ReturnValue)
   ) -> (Output_Get:bool)
   n6 = Get bIsCrouching()
   Set isWearingPackBool_Dliving(
    execute=@n1.then_4
    isWearingPackBool_Dliving:bool=AND Boolean!(A=Get isWearingPackBool(), B=Get bIsSkyDiving(), C:bool=NOT Boolean(A=@n6)) -> (ReturnValue)
   ) -> (Output_Get:bool)
   Set neck_rot_up(
    execute=@n1.then_1
    neck_rot_up:double=GetAlphaFromReferenceTransformDelta(
     SkeletalMeshComponent=Get Head_Mesh() -> (Head_Mesh:Skeletal Mesh Component Object Reference)
     BoneName=head
     BaseName=spine_05
     InputRange=(X=-20,Y=-80)
     TwistAxis="0,0,1"
    )
   ) -> (Output_Get:double)
   > Set neck_rot_r(neck_rot_r:double=GetAlphaFromReferenceTransformDelta(SkeletalMeshComponent=Get Head_Mesh(), BoneName=head, BaseName=spine_05, InputRange=(X=10,Y=20), TwistAxis="0,0,1")) -> (Output_Get:double)
   > Set neck_rot_l(neck_rot_l:double=GetAlphaFromReferenceTransformDelta(SkeletalMeshComponent=Get Head_Mesh(), BoneName=head, BaseName=spine_05, InputRange=(X=0,Y=20), TwistAxis="0,0,1")) -> (Output_Get:double)
   n7 = Event OnCharacterCustomizationLoaded()
   > n8 = Sequence!(execute) -> (then_0, then_1, then_2, then_3)
   Set Character Mesh(execute=@n8.then_1, Character Mesh:Skeletal Mesh Component Object Reference=@n7.CustomBaseMesh) -> (Output_Get:Skeletal Mesh Component Object Reference)
   n9 = CheckForBackbling()
   Set isWearingPackBool(execute=@n8.then_3, isWearingPackBool:bool=OR Boolean!(A=@n9, B=@n9.bIsBackblingACape, C=@n9.bIsBackblingLong) -> (ReturnValue)) -> (Output_Get:bool)
   > Set is_WearingCape(is_WearingCape:bool=@n9.bIsBackblingLong) -> (Output_Get:bool)
   Set Head_Mesh(
    execute=@n8.then_2
    Head_Mesh:Skeletal Mesh Component Object Reference=Get Array Item with EFort Custom Part Type(Array=@n7.CustomMeshParts, PartType=Head)
   ) -> (Output_Get:Skeletal Mesh Component Object Reference)
   GetSkeletalMesh!(self:Custom Character Part Object Reference=Get Array Item with EFort Custom Part Type(Array:Array of Custom Character Part Object References=@n7.CustomParts, PartType=Head) -> (Output:Custom Character Part Object Reference)) -> (ReturnValue:Skeletal Mesh Object Reference)
   AND Boolean!(A=@n9, B=NOT Boolean(A=@n9.bIsBackblingLong)) -> (ReturnValue)
   Set Body Mesh(Body Mesh:Skeletal Mesh Component Object Reference) -> (Output_Get:Skeletal Mesh Component Object Reference)
  ]}
]
variables [
 Body Mesh {type=SkeletalMeshComponent}
 Character Mesh {type=SkeletalMeshComponent}
 Gravity Override Cape {type=Vector}
 Gravity Override Params {
  type=GravityOverrideParamsStruct
  default {
   onGround_Standing {windAmplitude=<-10 -150 0>, windFrequency=<-7 -1 -1>, windOffset=<-10 -100 0>}
   skydive_DivingDown {jointName=root, gravityOverride=<0 1000 0>, windOffset=<0 0 50>}
   skydive_FloatingDown {gravityOverride=<0 -400 600>, windAmplitude=<50 -100 100>, windOffset=<0 -150 100>}
   skydive_GoingUp {gravityOverride=<0 -300 600>}
   skydive_Parachuting {gravityOverride=<0 -400 -800>, windAmplitude=<50 -100 500>, windOffset=<0 -150 300>}
   skydive_Surfing {gravityOverride=<0 -300 300>}
  }
 }
 Head_Mesh {type=SkeletalMeshComponent}
 Is Dive or Dive Floating or Wearing or DBNO {type=boolean}
 is_WearingCape {type=boolean}
 isWearingPackBool {type=boolean}
 IsWearingPackBool or DBNO or DBNOCarrying {type=number}
 isWearingPackBool_Dliving {type=boolean}
 neck_rot_l {type=number}
 neck_rot_r {type=number}
 neck_rot_up {type=number}
 Skel_Mesh {type=SkeletalMeshComponent}
 SkelMesh {type=SkeletalMeshComponent}
]
